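# Niagara cosmetic effect for the LiftingAura back bling, producing a persistent colored glow at a fixed offset on the backpack.
# One looping emitter burst-spawns sprites, using User.Variant to index a two-color array of red and magenta tints.
# Two sprite renderers and a light renderer are present, with a module deriving offset position and reduced-alpha color attributes for sprite use.
# Despite the emitter's name, the waveform flicker and light-attribute modules are disabled.
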NiagaraSystem /BRCosmetics/Accessories/FORT_Backpacks/Backpack_F_MED_LiftingAura/FX/NS_Backpack_LiftingAura
dependencies {usedDataInterfaces [&NiagaraDataInterfaceArrayColor], usedDynamicInputs [], usedModules [&SystemState]}
emitters {Default_LightAndGlowFlicker_Loop {
  emitterUpdateScript [
   &EmitterState(
    Inactive Response=::NewEnumerator1
    Life Cycle Mode=::NewEnumerator1
    Loop Behavior=::NewEnumerator1
    Loop Duration Mode=::NewEnumerator1
    UseLoopDelay=false
   )
   &SpawnBurst_Instantaneous(Use Spawn Probability=false, UseLoopCountLimit=false)
   /Niagara/Noise/Waveform(
    EnableExponent1=false
    GlobalClampMax=false
    GlobalClampMin=false
    UseInput1=false
    UsePhase1=false
    "Waveform [1]":ENiagara_Waveforms=::NewEnumerator12
    "[1] Frequency"=7
   ) {enabled=false}
   /SetVariables(Emitter.Flicker:NiagaraFloat=&Add_Float(A=&Multiply_Float(A=$Emitter.Waveform.Waveform, B=0.5), B=0.5)) {enabled=false, isSetParameters=true}
  ]
  particleSpawnScript [
   &InitializeParticle(
    Color=&SelectLinearColorFromArray(
     Array Sampling Mode=::NewEnumerator1
     Direct Array Index=&MakeIntFromFloat(Float=$User.Variant)
     Color Selection Array=DataInterface() {propertyValues {ColorData [<rgba 1 0 0.0233 0.8>, <rgba 1 0.0129 0.7626 0.8>]}}
     Random Seed=?
    )
    Position Offset=<0 -6.5 2.3>
    Sprite Size Mode=::NewEnumerator3
    Uniform Sprite Size=17
    UsePositionOffset=true
    Write Material Random=true
   )
   /SetVariables(Particles.SpriteFacing=<0 -1 0>) {isSetParameters=true}
   &DynamicMaterialParameters(
    Index 0 Param 1=1
    Index 0 Param 2=1
    Module.Param0WriteEnabled=true
    Module.Param1WriteEnabled=true
    Module.Param2WriteEnabled=true
    Module.Param3WriteEnabled=true
   ) {enabled=false}
  ]
  particleUpdateScript [
   &ParticleState(Kill Particles When Lifetime Has Elapsed=false, Let Infinitely Lived Particles Die When Emitter Deactivates=true)
   &ScaleColor(ScaleA=true, ScaleRGB=true) {enabled=false}
   &FN_Light_Attributes(
    Color=$Particles.Color
    Create Color Attribute=true
    Create Exponent Attribute=true
    Create Radius Attribute=true
    Exponent=5
    Radius=25
    Scale Color=0.5
   ) {enabled=false}
   &FN_Sprite_Attributes(
    Color=&MultiplyLinearColor(Linear Color=$Particles.Color, Scale Factor=<1 1 1 0.23>)
    Create Color Attribute=true
    Create Position Attribute=true
    Position:NiagaraPosition=&AddVectorToPosition(Position=$Particles.Position, Vector=<0 1.5 0>)
    Write to Module Namespace=true
   )
   &Color(Color=<rgba 1 1 1 10>) {enabled=false}
  ]
  rendererClasses [&NiagaraSpriteRendererProperties, &NiagaraLightRendererProperties]
  renderers [
   {rendererClass=&NiagaraSpriteRendererProperties}
   {rendererClass=&NiagaraLightRendererProperties, rendererIndex=1}
   {rendererClass=&NiagaraSpriteRendererProperties, rendererIndex=2}
  ]
 }}
summary {
 systemName=NS_Backpack_LiftingAura
 emitters [{
   emitterName=Default_LightAndGlowFlicker_Loop
   rendererClasses [&NiagaraSpriteRendererProperties, &NiagaraLightRendererProperties]
  }]
 userVariables [User.Variant {type=NiagaraFloat, defaultValue=0, description=""}]
}
