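# Animation Blueprint for the Lifting Aura backpack cosmetic that positions the pack and toggles its material elements.
# Each update maps the owner's spine_05-to-spine_01 yaw to an additive backpack roll and offset.
# Skydiving state blends between two bone-modifier chains on the cached input pose, one adding skydive and parachute offsets.
# Weapon, customization and skydive events set Hide Element 01 when a LiftingAura-tagged pickaxe is equipped outside skydiving.
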AnimBlueprint /BRCosmetics/Accessories/FORT_Backpacks/Backpack_F_MED_LiftingAura/Meshes/F_MED_LiftingAura_Pack_AnimBP
target_skeleton=&/F_MED_LiftingAura_Pack_Skeleton
graphs [
 AnimGraph {nodes [
   Input Pose()
   "Transform (Modify) Bone - Bone: pack"!(ComponentPose, bAlphaBoolEnabled:bool, Translation="0,0,0", Rotation="-0.000,0.000,27.478", Scale) -> (Pose) {
    alphaInputType=Bool
    rotationMode=BMM_Additive
    translationMode=BMM_Additive
    alphaBoolBlend {blendOption=Linear}
    boneToModify {boneName=pack}
    rotation {roll=27.478}
   }
   n1 = "Transform (Modify) Bone - Bone: pack"!(
    ComponentPose=Local To Component(LocalPose=Use cached pose 'cache pose'())
    Alpha:float
    Translation=Get Backpack Translation() -> (Backpack Translation:Vector)
    Rotation=Get Backpack Rotate() -> (Backpack Rotate:Rotator)
    Scale
   ) -> (Pose) {
    rotationMode=BMM_Additive
    rotationSpace=BCS_BoneSpace
    scaleMode=BMM_Additive
    scaleSpace=BCS_BoneSpace
    translationMode=BMM_Additive
    translationSpace=BCS_BoneSpace
    alphaBoolBlend {blendOption=Linear}
    boneToModify {boneName=pack}
   }
   n2 = "Transform (Modify) Bone - Bone: pack"!(ComponentPose=@n1, bAlphaBoolEnabled:bool=Get bIsSkyDiving(), Translation="0,-16,-10", Rotation="-45,0,-10", Scale) -> (Pose) {
    alphaInputType=Bool
    rotationMode=BMM_Additive
    rotationSpace=BCS_BoneSpace
    scaleMode=BMM_Additive
    scaleSpace=BCS_BoneSpace
    translationMode=BMM_Additive
    translationSpace=BCS_BoneSpace
    alphaBoolBlend {blendOption=Linear}
    boneToModify {boneName=pack}
    rotation {pitch=-45, roll=-10}
    translation {y=-16, z=-10}
   }
   n3 = "Transform (Modify) Bone - Bone: pack"(ComponentPose=@n2, bAlphaBoolEnabled=Get bIsParachuteOpened(), Translation="-2.620,-0.546,0.621", Rotation="40,-10,0") {
    alphaInputType=Bool
    rotationMode=BMM_Additive
    rotationSpace=BCS_BoneSpace
    scaleMode=BMM_Additive
    scaleSpace=BCS_BoneSpace
    translationMode=BMM_Additive
    translationSpace=BCS_BoneSpace
    alphaBoolBlend {blendOption=Linear}
    boneToModify {boneName=pack}
    rotation {pitch=40, yaw=-10}
    translation=<-2.62 -0.546 0.621>
   }
   n4 = "Transform (Modify) Bone - Bone: root"(
    ComponentPose=Local To Component(LocalPose=Use cached pose 'cache pose'())
    Translation=Get Backpack Translation()
    Rotation=Get Backpack Rotate()
   ) {
    rotationMode=BMM_Additive
    rotationSpace=BCS_BoneSpace
    scaleMode=BMM_Additive
    scaleSpace=BCS_BoneSpace
    translationMode=BMM_Additive
    translationSpace=BCS_BoneSpace
    alphaBoolBlend {blendOption=Linear}
    boneToModify {boneName=root}
   }
   n5 = "Transform (Modify) Bone - Bone: pack"!(ComponentPose=@n4, Alpha:float, Translation="0,-10,-9", Rotation="0,0,0", Scale) -> (Pose) {
    rotationMode=BMM_Additive
    rotationSpace=BCS_BoneSpace
    translationMode=BMM_Additive
    translationSpace=BCS_BoneSpace
    alphaBoolBlend {blendOption=Linear}
    boneToModify {boneName=pack}
    translation {y=-10, z=-9}
   }
   n6 = Blend Poses by bool(
    BlendPose_0=Component To Local(ComponentPose=@n3)
    BlendPose_1=Component To Local(ComponentPose=@n5)
    BlendTime_0=0.5
    bActiveValue=Get bIsSkyDiving()
   ) {blendType=Linear}
   Output Pose(Result=Component To Local(ComponentPose=Local To Component(LocalPose=@n6))) {layerGroup=None}
   Save cached pose 'cache pose'(Pose=Input Pose() {name=InPose_0}) {cacheName=cache pose}
  ]}
 EventGraph {nodes [
   n1 = GetPartSkeletalMeshComponent()
   n2 = Get Dynamic Material Instance(execute=[@n7.then, @n11.then], SkeletalMeshComponent=@n1)
   n3 = SetScalarParameterValue(self=@n2, ParameterName=Hide Element 10, Value=1)
   n4 = Get Dynamic Material Instance(execute=@n2.then, SkeletalMeshComponent=@n1, ElementIndex=1)
   SetScalarParameterValue(execute=@n3.then, self=@n4, ParameterName=HT_CrunchVerts, Value=1)
   SetScalarParameterValue(self=@n2, ParameterName=Hide Element 02, Value=1)
   > SetScalarParameterValue(self=@n4, ParameterName=HT_CrunchVerts, Value=0)
   Event BlueprintUpdateAnimation()
   > Sequence()
   n5 = GetOwnerSkeletalMeshComponent()
   n6 = BreakRotator(InRot=BreakTransform(InTransform=GetDeltaTransformFromRefPose(self=@n5, BoneName=spine_05, BaseName=spine_01)).Rotation)
   > Set Backpack Rotate(Backpack Rotate:Rotator=MakeRotator(Roll=MapRangeClamped(Value=@n6.Yaw, InRangeA=-12, InRangeB=-30, OutRangeA=3, OutRangeB=25))) -> (Output_Get:Rotator)
   > Set Backpack Translation(Backpack Translation:Vector=MakeVector(Y=MapRangeClamped(Value=@n6.Yaw, InRangeA=-12, InRangeB=-30, OutRangeB=-8))) -> (Output_Get:Vector)
   n7 = Event OnWeaponChanged()
   n8 = Event OnBeginSkydiving()
   SetScalarParameterValue(execute=[@n8.then, @n13.then, @n10.else, @n14.else], self=@n2, ParameterName=Hide Element 01, Value=0)
   Event OnEndSkydiving()
   n9 = PickaxeItemDefinitionHasAnyMetaTag(Tags=Make Array("[0]"=(TagName="Cosmetics.MetaTags.LiftingAura")), bRequiresEquipped=true)
   > n10 = Branch(Condition=@n9)
   SetScalarParameterValue(execute=[@n13.else, @n10.then], self=@n2, ParameterName=Hide Element 01, Value=1)
   n11 = Event OnCharacterCustomizationLoaded()
   n12 = Branch(Condition=Get bIsPlayingMeleeAnim())
   > Set Melee(Melee:double=1) -> (Output_Get:double)
   Set Melee(execute=@n12.else, Melee=0)
   TryGetPawnOwner()
   n13 = Branch(execute=@n14.then, Condition=Get bIsSkyDiving())
   n14 = Branch(execute=@n4.then, Condition=@n9)
  ]}
]
variables [
 Backpack Rotate {type=Rotator}
 Backpack Translation {type=Vector}
 Is Parachute Opened {type=boolean}
 Is Sky Diving {type=boolean}
 Melee {type=number}
 Not Sky Diving {type=boolean}
]
